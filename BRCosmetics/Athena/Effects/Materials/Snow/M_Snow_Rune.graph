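# Unlit translucent Niagara sprite and mesh material that draws glowing snow runes from a SubUV atlas selected by frame.
# Channel-masked rune shape, optional dynamic-parameter progress fill and a spherical background glow form one mask blending parameter or particle colors.
# Screen-space panning scanlines modulate color and opacity, with time-of-day HDR dimming at night.
# Camera, depth and view-angle fades shape opacity, with cheaper low-quality and mobile paths.
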Material /BRCosmetics/Athena/Effects/Materials/Snow/M_Snow_Rune
graph {
 properties {
  bUsedWithNiagaraMeshParticles=true
  bUsedWithNiagaraSprites=true
  blendMode=BLEND_Translucent
  pixelDepthOffsetMode=PDOM_Legacy
  shadingModel=MSM_Unlit
  usedShadingModels=Unlit
 }
 nodes [
  Constant() {r=0.5}
  MaterialFunctionCall() {materialFunction=&DebugTimeSine}
  n1 = ParticleColor()
  n2 = ScalarParameter() {defaultValue=1, parameterName=Particle Color Lerp Multiplier}
  n3 = VectorParameter() {parameterName=FillColor, defaultValue=<rgba 5 3.5523 0 0>}
  n4 = ScalarParameter() {defaultValue=-0.75, parameterName=MeterBottom}
  n5 = ScalarParameter() {defaultValue=-0.3, parameterName=MeterTop}
  n6 = DynamicParameter() {defaultValue=<rgba 1 1 1 0>, paramNames [Param1,Param2,ProgressAdd,Progress]}
  n7 = Add(A=@n6.ProgressAdd, B=@n6.ProgressAdd)
  n8 = Saturate(None=@n7)
  n9 = TextureSample(UVs=Add(A=TextureCoordinate(), B=LinearInterpolate(A=@n4, B=@n5, Alpha=@n8) {constA=-0.465, constAlpha=0, constB=-0.2})) {texture=/Game/Effects/Fort_Effects/Textures/Grad/T_Vertial_Gradient_Bar}
  n10 = Ceil(None=@n7)
  n11 = TextureObjectParameter() {parameterName=Texture, samplerType=SAMPLERTYPE_Normal, texture=/BRCosmetics/Athena/Effects/Textures/Snow/T_Snow_Runes}
  n12 = TextureCoordinate()
  n13 = AppendVector(A=ScalarParameter() {defaultValue=2, parameterName=Tile U}, B=ScalarParameter() {defaultValue=2, parameterName=Tile V})
  n14 = ScalarParameter() {parameterName=Frame}
  n15 = Clamp(None=ConstantBiasScale(None=MaterialFunctionCall!(Texture=@n11, UVs=@n12, SubImages=@n13, Frame=@n14) {materialFunction=/Engine/Functions/Engine_MaterialFunctions03/Texturing/SubUV_Function}.RGB))
  n16 = Multiply(A=Multiply(A=Multiply(A=@n3.RGB, B=@n9.G), B=Saturate(None=@n10)), B=ComponentMask(None=@n15) {g=true})
  n17 = Constant()
  n18 = StaticComponentMaskParameter(None=@n15) {defaultR=true, parameterName=Channel Mask}
  n19 = Add(A=ComponentMask(None=@n15) {r=true}, B=ComponentMask(None=@n15) {g=true})
  n20 = Multiply(A=SphereMask(A=TextureCoordinate(), B=Constant() {r=0.5}, Radius=Constant() {r=0.5}, Hardness=Constant() {r=0.75})) {constB=0.1}
  n21 = ComponentMask(None=TextureCoordinate()) {g=true}
  n22 = Multiply(A=Multiply(A=Power(Base=@n20) {constExponent=1.4}, B=Power(Base=@n21)), B=ScalarParameter() {parameterName=Background Shape Lerp})
  n23 = QualitySwitch(Default=@n22, Low=Constant(), High=@n22, Medium=@n22)
  n24 = Add(
   A=Add(
    A=StaticSwitchParameter(True=@n16, False=@n17) {parameterName=ProgressBar}
    B=StaticSwitchParameter(True=@n18, False=@n19) {defaultValue=true, parameterName=Is Single Masked?}
   )
   B=@n23
  )
  n25 = StaticSwitchParameter(
   True=LinearInterpolate(A=Multiply(A=@n1.RGB, B=@n2) {constB=0.5}, B=@n1.RGB, Alpha=@n24)
   False=LinearInterpolate(
    A=VectorParameter() {parameterName=Color, defaultValue=<rgba 0 0.6235 1 1>}.RGB
    B=VectorParameter() {parameterName=Color 2, defaultValue=<rgba 0 0.6235 1 1>}.RGB
    Alpha=@n24
   )
  ) {parameterName=Use Particle Color}
  n26 = Panner(Coordinate=Multiply(A=ScreenPosition().ViewportUV, B=ScalarParameter() {defaultValue=12, parameterName=PanLineTile}) {constB=4}, Time=Time()) {speedX=0.1, speedY=0.25}
  n27 = QualitySwitch(
   Default=Multiply(
    A=VectorParameter() {parameterName=ScanlineColor, defaultValue=<rgba 0.0562 0.1031 0.15 1>}.RGB
    B=TextureSample(UVs=@n26) {samplerSource=SSM_Wrap_WorldGroupSettings, texture=/Game/Effects/Fort_Effects/Textures/Tile/T_ScanLine_Analog}.R
   )
   Low=Constant()
  )
  n28 = Add(A=Multiply(A=@n27, B=@n25), B=@n25)
  n29 = ScalarParameter() {defaultValue=0.65, parameterName=Night Multiplier}
  n30 = MaterialFunctionCall(Color_Morning=@n28, Color_Day=@n28, Color_Evening=@n28, Color_Night=Multiply(A=@n28, B=@n29) {constB=0.45}) {materialFunction=&Color_Time_of_Day}
  n31 = ScalarParameter() {defaultValue=1, parameterName=HDRIntensity}
  n32 = MaterialFunctionCall(
   LDR Input=Clamp(None=@n30.Result)
   HDR Tint=@n30.Result
   HDR Intensity Multiplier=MaterialFunctionCall(Value_Morning=@n31, Value_Day=@n31, Value_Evening=@n31, Value_Night=Multiply(A=@n31, B=@n29) {constB=0.5}) {materialFunction=&Scalar_Time_of_Day}.Result
  ) {materialFunction=&DeriveHDRfromLDR}
  MaterialFunctionCall(Emissive=@n32.RemappedRange) {materialFunction=&MF_FX_EmissiveAdjustments}
  n33 = Multiply(A=@n27) {constB=4}
  n34 = ScalarParameter() {defaultValue=0.25, parameterName=Mask Multiplier}
  n35 = ScalarParameter() {defaultValue=45, parameterName=Distance Fade}
  n36 = Multiply(
   A=Add(A=Multiply(A=@n24, B=@n33), B=Multiply(A=@n24, B=@n34))
   B=Power(Base=MaterialFunctionCall(Distance=@n35) {materialFunction=&Near_Camera_Fade}.Result) {constExponent=3}
  )
  n37 = Multiply(A=@n36, B=DepthFade(FadeDistance=ScalarParameter() {defaultValue=20, parameterName=Depth Fade}))
  n38 = QualitySwitch(Default=@n37, Low=@n36, High=@n37, Medium=@n37)
  n39 = Multiply(
   A=@n38
   B=MaterialFunctionCall("FadeSensitivity (optional)"=ScalarParameter() {defaultValue=3, parameterName=Angle Fade Sensitivity}) {materialFunction=&FadeBasedOnViewAngle}.Result
  )
  Multiply(A=Add(A=ShadingPathSwitch(Default=@n39, Deferred=@n39, Mobile=@n38) {desc=Migrated from FeatureLevelSwitch}, B=@n23), B=ParticleColor().A)
  Reroute(None=@n32.RemappedRange)
 ]
}
